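; AITesting_MoveGoal is a move goal Blueprint for AI functional testing, built around a collision box that can be switched on or off.
; On BeginPlay it disables the box's collision when bStartEnabled is false.
; A SetCollisionEnabled function takes a boolean and sets the box to either QueryAndPhysics or NoCollision.
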(event EventBeginPlay
  (if (not (Variables|AITesting|GetbStartEnabled))
    (Collision|SetCollisionEnabled (Variables|MoveGoal|GetCollisionBox) "NoCollision")))

(fn SetCollisionEnabled
  (params (bShouldBeEnabled :type Boolean))
  (bind _collisionbox (Variables|MoveGoal|GetCollisionBox))
  (if bShouldBeEnabled
    (Collision|SetCollisionEnabled _collisionbox "QueryAndPhysics")
    (else
      (Collision|SetCollisionEnabled _collisionbox "NoCollision"))))

(fn ConstructionScript)
